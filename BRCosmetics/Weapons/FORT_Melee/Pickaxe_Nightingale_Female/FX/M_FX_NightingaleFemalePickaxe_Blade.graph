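# Unlit alpha-composite material for the blade effect on the female Nightingale pickaxe, usable on skeletal and Niagara meshes.
# Panned noise distorts a gradient texture that drives opacity and a color ramp, with scrolling smoothstepped bits and axis-aligned fresnel shaping the color.
# Emissive also involves glow, bright spots, a V-based hue shift and a dynamic-parameter-driven sphere-mask light at the pivot.
# Vertex red masks emissive and opacity; depth fade softens intersections; noise pushes vertices along normals.
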Material /BRCosmetics/Weapons/FORT_Melee/Pickaxe_Nightingale_Female/FX/M_FX_NightingaleFemalePickaxe_Blade
graph {
 properties {
  bUsedWithMeshDeformer=true
  bUsedWithNiagaraMeshParticles=true
  bUsedWithSkeletalMesh=true
  blendMode=BLEND_AlphaComposite
  pixelDepthOffsetMode=PDOM_Legacy
  shadingModel=MSM_Unlit
  usedShadingModels=Unlit
 }
 nodes [
  n1 = NamedRerouteUsage()
  n2 = ComponentMask(None=OneMinus(None=TextureCoordinate())) {g=true}
  n3 = ScalarParameter() {group=HueShift, parameterName=Hueshift}
  n4 = NamedRerouteUsage()
  n5 = NamedRerouteUsage()
  n6 = Add(A=@n5, B=Multiply(A=@n5, B=NamedRerouteUsage()))
  n7 = TextureCoordinate()
  n8 = AppendVector(
   A=ScalarParameter() {defaultValue=2, group=Distortion, parameterName=Distortion_Tile_U}
   B=ScalarParameter() {defaultValue=0.5, group=Distortion, parameterName=Distortion_Tile_V}
  )
  n9 = ScalarParameter() {defaultValue=-0.5, group=Distortion, parameterName=Distortion_Pan_U}
  n10 = ScalarParameter() {defaultValue=0.3, group=Distortion, parameterName=Distortion_Pan_V}
  n11 = TextureSample(UVs=Panner(Coordinate=Multiply(A=@n7, B=@n8), Speed=AppendVector(A=@n9, B=@n10)) {speedY=0.3}) {texture=&T_NoisePack_01}
  n12 = AppendVector(A=@n11.R, B=@n11.R)
  n13 = Constant2Vector() {g=1, r=2}
  n14 = VectorParameter() {parameterName=Distortion_Str, channelNames {a=Upper_Y, b=Upper_X, g=Lower_Y, r=Lower_X}, defaultValue=<rgba 0 0 0 1>}
  n15 = TextureCoordinate()
  n16 = Multiply(
   A=Add(A=Multiply(A=@n12, B=@n13) {constB=2}, B=Constant2Vector() {r=-1}) {constB=-1}
   B=LinearInterpolate(A=AppendVector(A=@n14.Upper_X, B=@n14.Upper_X), B=ComponentMask(None=@n14) {g=true, r=true}, Alpha=ComponentMask(None=@n15) {g=true})
  ) {constB=0.1}
  n17 = TextureSample(UVs=Add(A=TextureCoordinate(), B=@n16)) {
   samplerType=SAMPLERTYPE_LinearColor
   texture=&/T_FX_NightingaleF_Pickaxe_Gradient
  }
  n18 = VectorParameter() {parameterName=Glow, defaultValue=<rgba 0 0 0 1>}
  n19 = Add(
   A=MaterialFunctionCall(Hue Shift Percentage=Multiply(A=@n2, B=@n3), Texture=Add(A=@n4, B=@n6)) {materialFunction=&HueShift}.Result
   B=Multiply(A=Multiply(A=@n17.B, B=@n18.RGB), B=@n18.A)
  )
  n20 = VertexColor()
  MaterialFunctionCall(Emissive=Multiply(A=@n1, B=Multiply(A=@n19, B=@n20.R))) {materialFunction=&MF_FX_EmissiveAdjustments}
  n21 = TextureCoordinate()
  n22 = VectorParameter() {group=WPO, parameterName=VS_PushOutNoiseTransform, channelNames {a=SpeedY, b=SpeedX, g=ScaleY, r=ScaleX}, defaultValue=<rgba 0 0 0 1>}
  n23 = AppendVector(A=@n22.ScaleX, B=@n22.ScaleX)
  n24 = TextureSample(UVs=Panner(Coordinate=Multiply(A=@n21, B=@n23), Speed=AppendVector(A=@n22.SpeedX, B=@n22.SpeedX)) {speedY=0.3}) {texture=&T_NoisePack_01}
  n25 = ScalarParameter() {group=WPO, parameterName=VS_WPO_Push}
  Multiply(A=ComponentMask(None=TextureCoordinate()) {g=true}, B=Multiply(A=VertexNormalWS(), B=Multiply(A=@n24.G, B=@n25)))
  n26 = LinearInterpolate(A=ScalarParameter() {parameterName=MinOpacity}, Alpha=@n17.R) {constAlpha=1}
  Multiply(A=Multiply(A=Multiply(A=@n26, B=@n17.G), B=@n20.R), B=@n1)
  n27 = SphereMask(
   A=MaterialFunctionCall() {materialFunction=&ObjectPivotPoint}.Object Pivot Location
   B=WorldPosition().XYZ
   Radius=ScalarParameter() {group=InternalLight, parameterName=Internal_Light_Size}
  ) {hardnessPercent=0}
  n28 = ScalarParameter() {group=InternalLight, parameterName=Internal_Light_Falloff}
  NamedRerouteDeclaration(None=Multiply(
    A=Power(Base=@n27, Exp=@n28)
    B=DynamicParameter() {paramNames [Internal_Light_Brightness, Param2, Param3, Param4]}.Internal_Light_Brightness
   )) {name=InternalLight, nodeColor=<rgba 0 1 0.3672 1>}
  n29 = Multiply(A=@n17.R, B=ScalarParameter() {parameterName=Scale})
  n30 = Multiply(
   A=Multiply(
    A=@n17.G
    B=TextureSample(UVs=@n29) {texture=&/T_FX_NightingaleF_Pickaxe_ColorRamp}.RGB
   )
   B=ScalarParameter() {parameterName=Brightness}
  )
  n31 = TextureCoordinate()
  n32 = Add(A=OneMinus(None=ComponentMask(None=@n31) {g=true}), B=ScalarParameter() {parameterName=sstep_lag})
  n33 = ScalarParameter() {parameterName=ss}
  n34 = ScalarParameter() {defaultValue=20, group=Distortion, parameterName=Distortion_Intensity}
  n35 = VectorParameter() {parameterName=BitsTransform, channelNames {a=SpeedY, b=SpeedX, g=Y, r=X}, defaultValue=<rgba 0 0 0 1>}
  n36 = Multiply(A=TextureCoordinate(), B=AppendVector(A=@n35.X, B=@n35.X))
  n37 = AppendVector(A=@n35.SpeedX, B=@n35.SpeedX)
  n38 = TextureSample(UVs=Add(A=Multiply(A=@n16, B=@n34), B=Panner(Coordinate=@n36, Speed=@n37) {speedY=0.3})) {texture=/Game/Packages/Fortress_SharedMaterials/Textures/T_Ravens_SpecularReflection}
  n39 = Multiply(A=SmoothStep(Min=@n32, Max=Add(A=@n32, B=@n33), Value=@n38.R), B=VectorParameter() {parameterName=Bits, defaultValue=<rgba 0 0 0 1>}.RGB)
  n40 = ScalarParameter() {group=Fresnel, parameterName=BladeFresnelBrightness}
  n41 = NamedRerouteUsage()
  NamedRerouteDeclaration(None=Multiply(A=Add(A=@n30, B=@n39), B=LinearInterpolate(A=@n40, Alpha=@n41) {constAlpha=1})) {name=Blade_Pre, nodeColor=<rgba 0 1 0.5547 1>}
  NamedRerouteDeclaration(None=Multiply(A=@n38.R, B=VectorParameter() {group=Brightspots, parameterName=Brightspots, defaultValue=<rgba 0 0 0 1>}.RGB)) {name=BrightSpots, nodeColor=<rgba 0.3828 0 1 1>}
  n42 = ScalarParameter() {parameterName=Axis_Fresnel}
  NamedRerouteDeclaration(None=OneMinus(None=MaterialFunctionCall(Exponent=@n42) {materialFunction=&AxisAlignedFresnel}.Fresnel)) {name=BladeFresnel, nodeColor=<rgba 1 0.5391 0 1>}
  n43 = DepthFade(Opacity=Constant() {r=1}) {fadeDistanceDefault=25}
  NamedRerouteDeclaration(None=ShadingPathSwitch(Default=@n43, Deferred=@n43, Mobile=Constant() {r=1}) {desc=Migrated from FeatureLevelSwitch}) {name=IntersectFade, nodeColor=<rgba 0 1 0.6016 1>}
  ScalarParameter() {parameterName=BrightSpots}
 ]
}
