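# Additive, unlit Niagara sprite material for cartoon fire licks, part of the effects for the female Nightingale pickaxe.
# A SubUV flame texture's green channel serves as the mask, multiplied by particle color and alpha.
# The clamped color goes through DeriveHDRfromLDR, driven by a particle-colored HDR tint and intensity, then through shared FX emissive adjustments.
# A MasterAlpha parameter is present but feeds nothing in the graph.
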Material /BRCosmetics/Weapons/FORT_Melee/Pickaxe_Nightingale_Female/FX/M_PickaxeNightingaleF_ToonFireLicks
graph {
 properties {bUsedWithNiagaraSprites=true, blendMode=BLEND_Additive, pixelDepthOffsetMode=PDOM_Legacy, shadingModel=MSM_Unlit, usedShadingModels=Unlit}
 nodes [
  n1 = ParticleColor()
  n2 = ParticleSubUV!(UVs, Tex, Apply View MipBias) {texture=/Game/Effects/Fort_Effects/Textures/Fire/T_CartoonFlame}
  n3 = Multiply(A=@n1.RGB, B=@n2.G)
  n4 = VectorParameter() {group=Color, parameterName=HDR_Tint, sortPriority=2, defaultValue=<rgba 1 1 1 0>}
  MaterialFunctionCall(Emissive=MaterialFunctionCall(
    LDR Input=Clamp(None=@n3)
    HDR Tint=Multiply(A=@n1.RGB, B=@n4.RGB)
    HDR Intensity Multiplier=ScalarParameter() {defaultValue=1.5, group=Color, parameterName=HDR_Intensity, sortPriority=3}
   ) {materialFunction=&DeriveHDRfromLDR}.RemappedRange) {materialFunction=&MF_FX_EmissiveAdjustments}
  Saturate(None=Multiply(A=@n1.A, B=@n2.G))
  ScalarParameter() {defaultValue=1, group=Alpha, parameterName=MasterAlpha}
 ]
}
